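# Sniper reticle extension widget showing a two-level zoom toggle indicator with a keybind prompt.
# View target changes rebind it to the current pawn's ranged weapon and ToggleZoom component, whose target mode drives the circle animation.
# It fades in while targeting outside scoped third-person mode.
# Delta gates play animations forward or reverse only on state changes; newly bound components snap without animating.
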WidgetBlueprint /VividRazorWeaponGameplay/Gameplay/MarkExecuteSniper/UI/WBP_Reticle_SniperToggleZoom
animations [
 Anim_ToggleZoom {
  binding_count=3
  length_seconds=0.25
  object=&self:Anim_ToggleZoom
  sequence {bindings [
    Image_ZoomLevelToggle {
     id=#1
     possessed_class=Image
     tracks [
      {
       type=MovieScene2DTransformTrack
       sections [{
         type=MovieScene2DTransformSection
         end_frame=5
         start_frame=0
         channels [
          Translation.X {type=MovieSceneScriptingFloatChannel, keys [[0,0],[2.7777,0],[3.3333,-37]]}
          Scale.X {type=MovieSceneScriptingFloatChannel, keys [[0,1],[1.111,1],[2.7777,0],[3.3333,0.65]]}
          Scale.Y {type=MovieSceneScriptingFloatChannel, keys [[0,1],[1.111,1],[2.7777,0],[3.3333,0.65]]}
         ]
        }]
      }
      {
       type=MovieSceneColorTrack
       display_name=Color and Opacity
       sections [{
         type=MovieSceneColorSection
         end_frame=5
         start_frame=0
         channels [
          Color.R {type=MovieSceneScriptingFloatChannel, const [1,2]}
          Color.G {type=MovieSceneScriptingFloatChannel, const [1,2]}
          Color.B {type=MovieSceneScriptingFloatChannel, const [1,2]}
          Color.A {type=MovieSceneScriptingFloatChannel, keys [[0,0.65],[0.5557,0.6341],[2.7777,0],[3.3333,0.65]]}
         ]
        }]
      }
      {type=MovieSceneWidgetMaterialTrack, sections []}
     ]
    }
    Image_ZoomLevel_x2A {
     id=#2
     possessed_class=Image
     tracks [
      {
       type=MovieSceneColorTrack
       display_name=Color and Opacity
       sections [{
         type=MovieSceneColorSection
         end_frame=3
         start_frame=0
         channels [
          Color.R {type=MovieSceneScriptingFloatChannel, const [1,2]}
          Color.G {type=MovieSceneScriptingFloatChannel, const [1,2]}
          Color.B {type=MovieSceneScriptingFloatChannel, const [1,2]}
          Color.A {type=MovieSceneScriptingFloatChannel, keys [[0,0.3],[1.111,0.09],[3.889,0.09]]}
         ]
        }]
      }
      {
       type=MovieScene2DTransformTrack
       sections [{
         type=MovieScene2DTransformSection
         end_frame=1
         start_frame=0
         channels [
          Angle {type=MovieSceneScriptingFloatChannel, const [0,1]}
          Scale.X {type=MovieSceneScriptingFloatChannel, keys [[0,1],[1.111,1.5]]}
          Scale.Y {type=MovieSceneScriptingFloatChannel, keys [[0,1],[1.111,1.5]]}
          Shear.X {type=MovieSceneScriptingFloatChannel, const [0,1]}
          Shear.Y {type=MovieSceneScriptingFloatChannel, const [0,1]}
         ]
        }]
      }
      {
       type=MovieSceneWidgetMaterialTrack
       sections [{
         type=MovieSceneParameterSection
         end_frame=3
         start_frame=0
         channels [OuterRing_Blur {type=MovieSceneScriptingFloatChannel, keys [[0,0.05],[1.111,0.2],[3.889,0.2]]}]
        }]
      }
     ]
    }
    Image_ZoomLevel_x3B_1 {
     id=#3
     possessed_class=Image
     tracks [
      {
       type=MovieSceneColorTrack
       display_name=Color and Opacity
       sections [{
         type=MovieSceneColorSection
         end_frame=2
         start_frame=0
         channels [
          Color.R {type=MovieSceneScriptingFloatChannel, const [1,1]}
          Color.G {type=MovieSceneScriptingFloatChannel, const [1,1]}
          Color.B {type=MovieSceneScriptingFloatChannel, const [1,1]}
          Color.A {type=MovieSceneScriptingFloatChannel, keys [[0,0.2],[2.7777,0.4]]}
         ]
        }]
      }
      {
       type=MovieScene2DTransformTrack
       sections [{
         type=MovieScene2DTransformSection
         end_frame=1
         start_frame=0
         channels [
          Angle {type=MovieSceneScriptingFloatChannel, const [0,1]}
          Scale.X {type=MovieSceneScriptingFloatChannel, keys [[0,0.45],[1.111,0.65]]}
          Scale.Y {type=MovieSceneScriptingFloatChannel, keys [[0,0.45],[1.111,0.65]]}
          Shear.X {type=MovieSceneScriptingFloatChannel, const [0,1]}
          Shear.Y {type=MovieSceneScriptingFloatChannel, const [0,1]}
         ]
        }]
      }
      {
       type=MovieSceneWidgetMaterialTrack
       sections [{
         type=MovieSceneParameterSection
         end_frame=5
         start_frame=0
         channels [
          OuterRing_Blur {type=MovieSceneScriptingFloatChannel, keys [[0,0.21],[1.111,0.05]]}
          Outer_Thickness {type=MovieSceneScriptingFloatChannel, keys [[0,0.09],[1.111,0.06]]}
          Outer_Size {type=MovieSceneScriptingFloatChannel, keys [[0,0.79],[1.111,0.93]]}
         ]
        }]
      }
     ]
    }
   ]}
 }
 Anim_Intro {
  binding_count=1
  length_seconds=0.2539
  object=&self:Anim_Intro
  sequence {bindings [HorizontalBox_Root {
     id=#4
     possessed_class=HorizontalBox
     tracks [{
       type=MovieSceneFloatTrack
       display_name=Render Opacity
       sections [{type=MovieSceneFloatSection, channels [None {type=MovieSceneScriptingFloatChannel, keys [[0,0],[5.0787,1]]}]}]
      }]
    }]}
 }
]
graphs [
 EventGraph {nodes [
   Event Construct()
   > n1 = Sequence()
   > n2 = Cast To FortPlayerController(Object=GetOwningPlayer())
   > Assign On View Target Changed!(execute:Exec, self:Fort Player Controller Object Reference=@n2, Delegate:"Delegate (by ref)"=Create Event()) -> (then:Exec) {is=title K2Node_AssignDelegate}
   > OnViewTargetChanged!(
    execute
    self
    PlayerController:Fort Player Controller Object Reference
    OldViewTarget:Actor Object Reference
    NewViewTarget:Actor Object Reference=GetViewTarget!(self:Controller Object Reference=@n2) -> (ReturnValue:Actor Object Reference)
   ) -> (then)
   n3 = Branch(execute=@n1.then_1, Condition=GetDesiredVisibility!(self) -> (IsVisible:bool))
   > SetZoomVisibility!(execute, self, IsVisible:bool=true) -> (then)
   SetRenderOpacity(execute=@n3.else, self=Get HorizontalBox_Root() -> (HorizontalBox_Root:Horizontal Box Object Reference), InOpacity=0)
   Event OnTargetingChanged!() -> (OutputDelegate, then, bIsADS:bool) {is=title K2Node_Event}
   > SetZoomVisibility(IsVisible=GetDesiredVisibility())
   Event Destruct()
   > OnViewTargetChanged()
   > n4 = Cast To FortPlayerController(Object=GetOwningPlayer())
   > Unbind Event from On View Target Changed(self:Fort Player Controller Object Reference=@n4, Delegate=Create Event()) {is=title K2Node_RemoveDelegate}
  ]}
 GetDesiredVisibility {nodes [
   GetDesiredVisibility() {is=title K2Node_FunctionEntry}
   > n1 = Get!(execute:Exec) -> (then:Exec, else:Exec, BoundWeapon:Fort Weapon Ranged Object Reference)
   n2 = IsScopedThirdPersonTargetingModeEnabled!(self:Fort Weapon Ranged Object Reference=@n1) -> (ReturnValue:bool)
   > Return Node!(execute, IsVisible:bool=AND Boolean(A=IsTargeting!(self:Fort Weapon Object Reference=@n1) -> (ReturnValue:bool), B=NOT Boolean(A=@n2)))
   Return Node(execute=@n1.else, IsVisible=false)
  ]}
 IsCurrentlyZoomed {nodes [
   IsCurrentlyZoomed() {is=title K2Node_FunctionEntry}
   n1 = Get BoundToggleZoomComponent() -> (BoundToggleZoomComponent:Fort Weapon Component Toggle Zoom Object Reference)
   > n2 = Is Valid(InputObject=@n1)
   Return Node!(execute=@n2.Is Not Valid, IsZoomed:bool=false)
   > Return Node(IsZoomed="integer > integer"(A=Get TargetMode!(self:Fort Weapon Component Target Mode Selector Object Reference=@n1) -> (TargetMode:int=0)))
  ]}
 OnScopedThirdPersonTargetingChanged {nodes [
   OnScopedThirdPersonTargetingChanged!() -> (then, bNewIsScopedThirdPersonTargeting:bool=~) {is=title K2Node_FunctionEntry}
   > SetZoomVisibility(IsVisible=GetDesiredVisibility() {is=title K2Node_CallFunction})
  ]}
 OnTargetModeChanged {nodes [
   OnTargetModeChanged!() -> (then, TargetMode:int) {is=title K2Node_FunctionEntry}
   > SetZoomActive!(execute, self, IsZoomActive:bool=IsCurrentlyZoomed!(self) -> (IsZoomed:bool) {is=title K2Node_CallFunction}, PlayAnimation:bool=true) -> (then)
  ]}
 OnViewTargetChanged {nodes [
   n1 = OnViewTargetChanged!() -> (then, PlayerController:Fort Player Controller Object Reference, OldViewTarget:Actor Object Reference, NewViewTarget:Actor Object Reference) {is=title K2Node_FunctionEntry}
   > n2 = Cast To FortPlayerPawn(Object=@n1.NewViewTarget)
   n3 = GetCurrentWeapon!(self:Fort Pawn Object Reference=@n2) -> (ReturnValue:Fort Weapon Object Reference)
   > n4 = Is Valid(InputObject=@n3)
   UnbindToggleZoomComponentEvents(execute=[@n4.Is Not Valid, @n2.CastFailed])
   > UnbindWeaponEvents()
   SetToggleZoomComponent!(
    execute=@n4.Is Valid
    self
    NewComponent:Fort Weapon Component Toggle Zoom Object Reference=GetComponentByClass!(
     self:Actor Object Reference=@n3
     ComponentClass:Actor Component Class Reference=/VividRazorWeaponGameplay/Gameplay/SniperToggleZoom/ToggleZoomOptic/FortWeaponComponent_ToggleZoom.FortWeaponComponent_ToggleZoom_C
    ) -> (ReturnValue:Fort Weapon Component Toggle Zoom Object Reference)
   ) -> (then)
   > SetWeapon!(execute, self, NewWeapon:Fort Weapon Object Reference=@n3) -> (then)
  ]}
 SetToggleZoomComponent {nodes [
   n1 = SetToggleZoomComponent!() -> (then, NewComponent:Fort Weapon Component Toggle Zoom Object Reference) {is=title K2Node_FunctionEntry}
   > Branch(Condition="Not Equal (Object)"(A:Object Reference=@n1, B:Object Reference=Get BoundToggleZoomComponent()) -> (ReturnValue:bool=~) {is=title K2Node_PromotableOperator})
   > UnbindToggleZoomComponentEvents()
   > n2 = Set BoundToggleZoomComponent(BoundToggleZoomComponent:Fort Weapon Component Toggle Zoom Object Reference=Get NewComponent() -> (NewComponent:Fort Weapon Component Toggle Zoom Object Reference)) -> (Output_Get:Fort Weapon Component Toggle Zoom Object Reference)
   > Is Valid(InputObject=@n2)
   > Bind Event to On Target Mode Changed(self:Fort Weapon Component Target Mode Selector Object Reference=Get BoundToggleZoomComponent(), Delegate=Create Event()) {is=title K2Node_AddDelegate}
   > SetZoomActive(IsZoomActive=IsCurrentlyZoomed(), PlayAnimation=false)
  ]}
 SetWeapon {nodes [
   n1 = SetWeapon!() -> (then, NewWeapon:Fort Weapon Object Reference) {is=title K2Node_FunctionEntry}
   > n2 = Cast To FortWeaponRanged!(execute, Object=@n1) -> (then, CastFailed, AsFort Weapon Ranged:Fort Weapon Ranged Object Reference) {is=title K2Node_DynamicCast}
   > Branch(Condition="Not Equal (Object)"(A=@n2, B=Get BoundWeapon() -> (BoundWeapon:Fort Weapon Ranged Object Reference)) -> (ReturnValue=~))
   > UnbindWeaponEvents()
   > n3 = Set BoundWeapon(BoundWeapon:Fort Weapon Ranged Object Reference=@n2) -> (Output_Get:Fort Weapon Ranged Object Reference)
   > Bind Event to On Scoped Third Person Targeting Changed(self:Fort Weapon Ranged Object Reference=@n3, Delegate=Create Event()) {is=title K2Node_AddDelegate}
   > OnScopedThirdPersonTargetingChanged!(execute, self, bNewIsScopedThirdPersonTargeting:bool=IsScopedThirdPersonTargetingModeEnabled(self=Get BoundWeapon())) -> (then) {is=title K2Node_CallFunction}
   UnbindWeaponEvents(execute=@n2.CastFailed)
  ]}
 SetZoomActive {nodes [
   n1 = SetZoomActive!() -> (then, IsZoomActive:bool=~, PlayAnimation:bool=true) {is=title K2Node_FunctionEntry}
   > n2 = Delta Gate Wildcard!(execute:Exec, Variable:"Boolean (by ref)"=Get IsZoomActive() -> (IsZoomActive:bool), New Value:bool=@n1) -> (Success:Exec, Updated Variable:"Boolean (by ref)", Previous Variable:bool=~) {is=title K2Node_MacroInstance}
   > n3 = Branch(Condition=Get PlayAnimation() -> (PlayAnimation:bool))
   > n4 = Branch(Condition=@n2)
   n5 = Get Anim_ToggleZoom() -> (Anim_ToggleZoom:Widget Animation Object Reference)
   > PlayAnimationReverse(InAnimation=@n5)
   PlayAnimationForward(execute=@n4.else, InAnimation=@n5)
   n6 = Get Anim_ToggleZoom()
   PlayAnimation(
    execute=@n3.else
    InAnimation=@n6
    StartAtTime=GetEndTime(self=@n6)
    PlayMode=Select(Option 0:EUMGSequencePlayMode Enum, Option 1:EUMGSequencePlayMode Enum=Reverse, Index=Get IsZoomActive()) -> (ReturnValue:EUMGSequencePlayMode Enum)
   )
  ]}
 SetZoomVisibility {nodes [
   n1 = SetZoomVisibility!() -> (then, IsVisible:bool=~) {is=title K2Node_FunctionEntry}
   > n2 = Delta Gate Wildcard(Variable=Get IsZoomVisible() -> (IsZoomVisible:bool), New Value=@n1) -> (Previous Variable=~)
   > n3 = Branch(Condition=@n2)
   n4 = Get Anim_Intro() -> (Anim_Intro:Widget Animation Object Reference)
   > PlayAnimationForward(InAnimation=@n4)
   PlayAnimationReverse(execute=@n3.else, InAnimation=@n4)
  ]}
 UnbindToggleZoomComponentEvents {nodes [
   UnbindToggleZoomComponentEvents() {is=title K2Node_FunctionEntry}
   n1 = Get BoundToggleZoomComponent()
   > Is Valid(InputObject=@n1)
   > Unbind Event from On Target Mode Changed(self:Fort Weapon Component Target Mode Selector Object Reference=@n1, Delegate=Create Event()) {is=title K2Node_RemoveDelegate}
   > Set BoundToggleZoomComponent()
  ]}
 UnbindWeaponEvents {nodes [
   UnbindWeaponEvents() {is=title K2Node_FunctionEntry}
   > n1 = Get()
   > Unbind Event from On Scoped Third Person Targeting Changed(self:Fort Weapon Ranged Object Reference=@n1, Delegate=Create Event()) {is=title K2Node_RemoveDelegate}
   > Set BoundWeapon()
  ]}
]
variables [
 BoundToggleZoomComponent {type=FortWeaponComponent_ToggleZoom_C, default=null}
 BoundWeapon {type=FortWeaponRanged, default=null}
 IsZoomActive {type=boolean, default=true}
 IsZoomVisible {type=boolean}
]
widgets {
 info {
  inheritedWidgetCount=1
  parentClass=/Script/FortniteUI.FortWeaponReticleExtensionWidgetBase
  rootWidgetClass=&HorizontalBox
  widgetCount=9
 }
 widgets [
  HorizontalBox_Root {
   widgetClassPath=&HorizontalBox
   bIsVariable=true
   widget=HorizontalBox_Root
   properties {renderTransform {angle=0, scale=<1 1>, shear=<0 0>, translation=<0 200>}}
  }
  Image_ZoomLevel_x3B_1 {
   widgetClassPath=&Image
   parent=HorizontalBox_Root
   widget=Image_ZoomLevel_x3B_1
   properties {
    brush {
     bIsDynamicallyLoaded=false
     drawAs=Image
     imageType=NoImage
     mirroring=NoMirror
     resourceName=None
     resourceObject=/ToggleZoomDMRGameplay/Items/ToggleZoomDMR/UI/Materials/MI_UI_ToggleZoomDMR_ToggleCircle_EmptyB
     tiling=NoTile
     imageSize=<32 32>
     margin=<ltrb 0 0 0 0>
     outlineSettings {
      bUseBrushTransparency=false
      roundingType=HalfHeightRadius
      width=0
      color {colorUseRule=UseColor_Specified, specifiedColor=<rgba 0 0 0 0>}
      cornerRadii=<0 0 0 0>
     }
     tintColor {colorUseRule=UseColor_Specified, specifiedColor=<rgba 0 1 0.1356 1>}
     uVRegion {bIsValid=false, max=<0 0>, min=<0 0>}
    }
    colorAndOpacity=<rgba 1 1 1 0.7>
    renderTransform {angle=0, scale=<0.65 0.65>, shear=<0 0>, translation=<0 0>}
   }
   slot {class=HorizontalBoxSlot}
  }
  Spacer_45 {widgetClassPath=&Spacer, parent=HorizontalBox_Root, widget=Spacer_45, properties {size=<5 1>}, slot {class=HorizontalBoxSlot}}
  Overlay_1 {widgetClassPath=&Overlay, parent=HorizontalBox_Root, widget=Overlay_1, slot {class=HorizontalBoxSlot}}
  Image_ZoomLevelToggle {
   widgetClassPath=&Image
   parent=Overlay_1
   widget=Image_ZoomLevelToggle
   properties {
    brush {
     bIsDynamicallyLoaded=false
     drawAs=Image
     imageType=NoImage
     mirroring=NoMirror
     resourceName=None
     resourceObject=/ToggleZoomDMRGameplay/Items/ToggleZoomDMR/UI/Materials/MI_UI_ToggleZoomDMR_ToggleCircle_Filled
     tiling=NoTile
     imageSize=<32 32>
     margin=<ltrb 0 0 0 0>
     outlineSettings {
      bUseBrushTransparency=false
      roundingType=HalfHeightRadius
      width=0
      color {colorUseRule=UseColor_Specified, specifiedColor=<rgba 0 0 0 0>}
      cornerRadii=<0 0 0 0>
     }
     tintColor {colorUseRule=UseColor_Specified, specifiedColor=<rgba 0.1388 1 0.1054 1>}
     uVRegion {bIsValid=false, max=<0 0>, min=<0 0>}
    }
    colorAndOpacity=<rgba 1 1 1 0.65>
   }
   slot {class=OverlaySlot}
  }
  Image_ZoomLevel_x2A {
   widgetClassPath=&Image
   parent=Overlay_1
   widget=Image_ZoomLevel_x2A
   properties {
    brush {
     bIsDynamicallyLoaded=false
     drawAs=Image
     imageType=NoImage
     mirroring=NoMirror
     resourceName=None
     resourceObject=/ToggleZoomDMRGameplay/Items/ToggleZoomDMR/UI/Materials/MI_UI_ToggleZoomDMR_ToggleCircle_Empty
     tiling=NoTile
     imageSize=<32 32>
     margin=<ltrb 0 0 0 0>
     outlineSettings {
      bUseBrushTransparency=false
      roundingType=HalfHeightRadius
      width=0
      color {colorUseRule=UseColor_Specified, specifiedColor=<rgba 0 0 0 0>}
      cornerRadii=<0 0 0 0>
     }
     tintColor {colorUseRule=UseColor_Specified, specifiedColor=<rgba 0 1 0.1356 1>}
     uVRegion {bIsValid=false, max=<0 0>, min=<0 0>}
    }
    colorAndOpacity=<rgba 1 1 1 0.3>
   }
   slot {class=OverlaySlot, properties {horizontalAlignment=HAlign_Fill, verticalAlignment=VAlign_Fill}}
  }
  FortVisualAttachment_0 {
   widgetClassPath=&FortVisualAttachment
   parent=HorizontalBox_Root
   widget=FortVisualAttachment_0
   properties {contentAnchor=<0 0.5>, renderTransform {angle=0, scale=<1 1>, shear=<0 0>, translation=<20 0>}}
   slot {class=HorizontalBoxSlot, properties {verticalAlignment=VAlign_Center}}
  }
  Keybind_Toggle {
   widgetClassPath=/FortUILibrary/Input/KeybindWidget
   parent=FortVisualAttachment_0
   widget=Keybind_Toggle
   properties {
    boundInputAction=/ToggleScopeModeInput/Input/IA_ToggleScopeMode
    forcedHoldKeybindStatus=ForcedHold
    renderTransform {angle=0, scale=<0.8 0.8>, shear=<0 0>, translation=<0 0>}
    touchOverrideImageSet {
     bUseDataRegistryId=false
     dataRegistryId=""
     dataTableRow {dataTable=/FortniteCoreUI/Foundation/RichTextData/CommonRichTextImages, rowName=None}
    }
   }
   slot {class=SizeBoxSlot}
  }
  TutorialHighlight {widgetClassPath=&TutorialHighlight}
 ]
}
